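# M_GrindWire is a Default Lit material for grind wire props among the Helios environment grindables, with defaults for a dark grey, fairly metallic surface.
# It exposes Color, Metallic, Specular and Roughness as parameters for material instances, with no textures or math nodes.
# Usage flags enable it on Nanite and spline meshes.
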Material /Game/Environments/Helios/Props/Grindables/Materials/M_GrindWire
graph {
 properties {bUsedWithNanite=true, bUsedWithSplineMeshes=true, pixelDepthOffsetMode=PDOM_Legacy, usedShadingModels=Default Lit}
 nodes [
  ScalarParameter() {defaultValue=0.45, parameterName=Roughness}
  ScalarParameter() {defaultValue=0.1, parameterName=Specular}
  ScalarParameter() {defaultValue=0.91, parameterName=Metallic}
  VectorParameter() {parameterName=Color, defaultValue=<rgba 0.128 0.128 0.128 0>}
 ]
}
