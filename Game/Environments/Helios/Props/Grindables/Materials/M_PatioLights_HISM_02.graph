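# Two-sided Nanite material for patio string-light bulbs on grindable props, supporting instanced static and spline meshes.
# Diffuse is tinted warm orange, scaled by time of day and inverse Fresnel, and masked by the Spec texture for glow.
# Foliage-style sway logic combines wind actor data, vertex-color offsets, a bend lookup texture and physics-field response.
# A world-position-offset branch rotates vertices about an axis around a per-instance pivot.
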Material /Game/Environments/Helios/Props/Grindables/Materials/M_PatioLights_HISM_02
graph {
 properties {
  bUsedWithInstancedStaticMeshes=true
  bUsedWithNanite=true
  bUsedWithSplineMeshes=true
  pixelDepthOffsetMode=PDOM_Legacy
  twoSided=true
  usedShadingModels=Default Lit
 }
 nodes [
  n1 = NamedRerouteUsage()
  n2 = MaterialFunctionCall(Float3=@n1) {materialFunction=&BreakOutFloat3Components}
  n3 = MaterialFunctionCall(X=@n2.G, Y=@n2.G, Z=@n2.G) {materialFunction=&MakeFloat3}
  n4 = Length(None=@n1) {is=class MaterialExpressionLength}
  n5 = RotateAboutAxis(
   NormalizedRotationAxis=Transform(None=Normalize(VectorInput=@n3.Result)) {transformSourceType=TRANSFORMSOURCE_Local}
   RotationAngle=Power(Base=Multiply(A=@n4) {constB=0.004})
   PivotPoint=TransformPosition(None=Constant3Vector() {constant=<rgba 0 0 5 0>}) {transformSourceType=TRANSFORMPOSSOURCE_Instance, transformType=TRANSFORMPOSSOURCE_World}
   Position=WorldPosition().XYZ
  )
  n6 = Add(A=@n5, B=WorldPosition() {worldPositionShaderOffset=WPT_ExcludeAllShaderOffsets}.XYZ)
  n7 = TransformPosition(None=@n6) {transformSourceType=TRANSFORMPOSSOURCE_World, transformType=TRANSFORMPOSSOURCE_Instance}
  n8 = MaterialFunctionCall() {materialFunction=&ObjectPivotPoint}
  Add(A=@n5, B=Subtract(A=Add(A=@n7, B=@n8.Object Pivot Location), B=@n6))
  Add(A=@n5)
  Divide() {constB=200}
  n9 = Multiply(A=NamedRerouteUsage(), B=ScalarParameter() {defaultValue=1, group=AnimationTweaking, parameterName=Overall_MaxBend_0to1})
  n10 = NamedRerouteUsage()
  n11 = ScalarParameter() {defaultValue=1, group=Animation Physics, parameterName=Overall_PhysicsField_MaxBend}
  n12 = StaticSwitch(True=LinearInterpolate(A=@n9, B=Multiply(A=@n10, B=@n11), Alpha=NamedRerouteUsage()), False=@n9, Value=NamedRerouteUsage())
  LinearInterpolate(A=@n12, B=ScalarParameter() {defaultValue=1, group=Animation, parameterName=Overall_PlayerBend}, Alpha=NamedRerouteUsage())
  MaterialFunctionCall!(
   Previous WPO Value
   Force Magnitude Mask
   Force Magnitude Multiplier
   Project Field Force Vector to Normals
   Noise Size
   Noise Pan Speed
   Noise Falloff
   Stretch Noise Coordinates Z
  ) {desc=Sample global physics field add to WPO, materialFunction=/Game/Effects/Niagara/MasterMaterials/Functions/MF_FX_PhysicsFields_Grass}
  n13 = Constant() {r=0.5}
  n14 = Multiply(
   A=Constant3Vector() {constant=<rgba 20 9.9456 3.3946 1>}
   B=MaterialFunctionCall(Value_Morning=@n13, Value_Day=Constant() {r=1}, Value_Evening=@n13, Value_Night=@n13) {materialFunction=&Scalar_Time_of_Day}.Result
  )
  n15 = OneMinus(None=Fresnel() {exponent=0.25})
  n16 = TextureSampleParameter2D() {parameterName=Spec, samplerType=SAMPLERTYPE_Masks, texture=/Game/Environments/Helios/Props/Grindables/Textures/T_Grindables_LightBulbs_01_S}
  Multiply(
   A=Multiply(
    A=TextureSampleParameter2D() {parameterName=Diffuse, texture=/Game/Environments/Helios/Props/Grindables/Textures/T_Grindables_LightBulbs_01_D}.RGB
    B=Multiply(A=@n14, B=@n15)
   )
   B=ComponentMask(None=@n16.A) {g=true}
  )
  NamedRerouteDeclaration(None=@n5) {name=WPO, nodeColor=<rgba 1 0.2578 0 1>}
  n17 = Constant3Vector()
  n18 = Multiply(
   A=NamedRerouteDeclaration(None=TransformPosition(None=@n17) {transformSourceType=TRANSFORMPOSSOURCE_Instance, transformType=TRANSFORMPOSSOURCE_World}) {name=BasePivot, nodeColor=<rgba 1 0 0.7031 1>}
   B=ScalarParameter() {defaultValue=1408.6052, group=AnimationTweaking, parameterName=Overall_Bend_Frequency}
  )
  n19 = StaticBool() {value=true}
  n20 = MaterialFunctionCall!(
   Position=@n18
   IsTrees
   IsGrass
   IsMediumPlant=@n19
   IsMediumPlant_2024Anim
   IsMediumPlant2ndLevel=StaticBool() {value=true}
   MediumPlant_BasePivotPostion=NamedRerouteUsage()
  ) {materialFunction=/Game/Environments/Asteria/Foliage/Wind/MF_WindActorData}
  NamedRerouteDeclaration(None=@n20.WindI_Direction_WithNoise) {name=WindDirectionOverallBend, nodeColor=<rgba 0 0.1562 1 1>}
  NamedRerouteDeclaration(None=@n20.WindIntesnity) {name=WindIntensityNoise_Overall, nodeColor=<rgba 1 0 0.4219 1>}
  n21 = Multiply(
   A=ComponentMask(None=TextureCoordinate() {coordinateIndex=1}) {g=true}
   B=ScalarParameter() {defaultValue=1308.7753, group=AnimationTweaking, parameterName=Leaf_RandomFrequency}
  )
  n22 = ScalarParameter() {defaultValue=246.1449, group=AnimationTweaking, parameterName=Leaf_DragWiggle_Calm}
  n23 = ScalarParameter() {defaultValue=120, group=AnimationTweaking, parameterName=Leaf_DragWiggle_Storm}
  n24 = NamedRerouteUsage()
  n25 = Add(A=@n21, B=Multiply(A=VertexColor().A, B=LinearInterpolate(A=@n22, B=@n23, Alpha=@n24)))
  n26 = AppendVector(A=@n25, B=@n25)
  n27 = MaterialFunctionCall(Position=Add(A=AppendVector(A=@n26, B=@n21), B=@n18), IsMediumPlant=@n19) {materialFunction=&MF_WindActorData}
  NamedRerouteDeclaration(None=@n27.StormMask) {name=StormMask, nodeColor=<rgba 0.3125 0 1 1>}
  n28 = VertexColor()
  n29 = TextureSample(UVs=AppendVector(A=@n12, B=OneMinus(None=@n28.A))) {
   samplerSource=SSM_Clamp_WorldGroupSettings
   samplerType=SAMPLERTYPE_LinearColor
   texture=/Game/Environments/Asteria/Foliage/Grass/MasterMaterial/GenericShared_Grass_Foliage_Anim
  }
  n30 = Multiply(A=@n29.R, B=ScalarParameter() {defaultValue=1.3, group=AnimationTweaking, parameterName=Overall_Bend_XY})
  n31 = NamedRerouteUsage()
  n32 = StaticSwitch(True=LinearInterpolate(A=@n31, B=NamedRerouteUsage(), Alpha=NamedRerouteUsage()), False=@n31, Value=NamedRerouteUsage())
  n33 = Constant2Vector()
  n34 = Multiply(A=Multiply(A=@n29.G) {constB=-1}, B=ScalarParameter() {defaultValue=1.3, group=AnimationTweaking, parameterName=Overall_Bend_Z})
  NamedRerouteDeclaration(None=Add(A=Multiply(A=@n30, B=@n32), B=AppendVector(A=@n33, B=@n34))) {name=Cheap OverallBend, nodeColor=<rgba 0.5 1 0 1>}
  NamedRerouteDeclaration(None=StaticBool() {value=true}) {name=Use Physics, nodeColor=<rgba 0 1 0.3438 1>}
  n35 = SamplePhysicsVectorField!(World Position=NamedRerouteUsage()) {fieldTarget=Vector_LinearVelocity}
  n36 = MaterialFunctionCall(Vector 3=Add(A=@n35) {constB=0.0001}) {materialFunction=/Engine/Functions/Engine_MaterialFunctions02/Utility/VectorLength}
  n37 = Constant()
  NamedRerouteDeclaration(None=ShadingPathSwitch(Default=Saturate(None=@n36.V3 Length), Mobile=@n37) {desc=Migrated from FeatureLevelSwitch}) {name=PhysicsFieldMask, nodeColor=<rgba 0 1 0.3906 1>}
  n38 = Constant3Vector() {constant=<rgba 0 0 1 0>}
  NamedRerouteDeclaration(None=ShadingPathSwitch(
    Default=MaterialFunctionCall(Vector=@n35, Default=@n38) {materialFunction=&SafeNormalize}.Result
    Mobile=Constant3Vector() {constant=<rgba 0 0 1 0>}
   ) {desc=Migrated from FeatureLevelSwitch}) {name=PhysicsFieldDirection, nodeColor=<rgba 0.3828 0 1 1>}
  n39 = Min(A=@n36.V3 Length, B=ScalarParameter() {defaultValue=1, group=PhysicsField, parameterName=MaxRotation_PhysicsField_Clamp})
  n40 = Time()
  n41 = VertexColor()
  n42 = Add(A=@n41.R, B=@n41.R)
  n43 = Multiply(
   A=ScalarParameter() {defaultValue=351, group=PhysicsField, parameterName=PhysicsField_NoiseOffset}
   B=Multiply(A=Add(A=@n42, B=@n41.B)) {constB=33.33}
  )
  n44 = Frac(None=Multiply(A=ScalarParameter() {defaultValue=2.5, group=PhysicsField, parameterName=MaxRotation_PhysicsField_NoiseSpeed}, B=Add(A=@n40, B=@n43)))
  n45 = OneMinus(None=@n44)
  n46 = LinearInterpolate(
   A=Multiply(A=Multiply(A=@n44, B=@n45)) {constB=2}
   Alpha=ScalarParameter() {defaultValue=0.6, group=PhysicsField, parameterName=MaxRotation_PhysicsField_MInBend}
  )
  NamedRerouteDeclaration(None=ShadingPathSwitch(Default=Multiply(A=@n39, B=@n46), Mobile=@n37) {desc=Migrated from FeatureLevelSwitch}) {name=PhysicsFieldStrength, nodeColor=<rgba 1 0 0.2109 1>}
  NamedRerouteDeclaration(None=@n27.WindIntensitySingleSample) {name=WindIntensityNoise, nodeColor=<rgba 0.1719 0 1 1>}
  NamedRerouteDeclaration(None=@n27.WindI_Direction_WithNoise) {name=WindDirectionWithNoise, nodeColor=<rgba 0 1 0.9297 1>}
  NamedRerouteUsage()
  NamedRerouteUsage()
  Reroute(None=@n27.StormMask)
  n47 = TextureSampleParameter2D() {
   parameterName=Normal
   samplerType=SAMPLERTYPE_Normal
   texture=/Game/Environments/Helios/Props/Grindables/Textures/T_Grindables_LightBulbs_01_N
  }
  Transform(None=@n47.RGB) {transformType=TRANSFORM_Local}
 ]
}
